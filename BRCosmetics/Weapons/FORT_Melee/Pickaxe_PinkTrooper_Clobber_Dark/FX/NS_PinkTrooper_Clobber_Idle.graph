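# Idle Niagara effect for the PinkTrooper Clobber Dark pickaxe, flashing short-lived glints on the weapon.
# One emitter bursts on a randomized one-to-two-second loop, placing particles via skeletal mesh sampling of User.Pickaxe plus a small rotation-aligned shape offset, nudged by a camera offset.
# Three sprite renderers form a layered flare that grows while fading from an intense HDR spike, with separate corona and soft-flare attributes.
# User.LightVersion blends colors from blue toward pale blue-white.
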NiagaraSystem /BRCosmetics/Weapons/FORT_Melee/Pickaxe_PinkTrooper_Clobber_Dark/FX/NS_PinkTrooper_Clobber_Idle
dependencies {usedDataInterfaces [&NiagaraDataInterfaceCurve], usedDynamicInputs [], usedModules [&SystemState]}
emitters {FlareComplex {
  emitterUpdateScript [
   &EmitterState(
    Life Cycle Mode=::NewEnumerator1
    Loop Duration=&RandomRangeFloat(Minimum=1, Maximum=2)
    Recalculate Duration Each Loop=true
    UseLoopDelay=false
   )
   &SpawnBurst_Instantaneous(Use Spawn Probability=false, UseLoopCountLimit=false)
  ]
  particleSpawnScript [
   &InitializeParticle(
    Color=&LerpLinearColors(StartColor=<rgba 0.0642 0.3294 1 1>, EndColor=<rgba 0.5885 0.9109 1 1>, LerpFactor=$User.LightVersion)
    Lifetime=&RandomRangeFloat(Minimum=0.1, Maximum=0.3)
    Sprite Rotation Mode=::NewEnumerator2
    Sprite Size Mode=::NewEnumerator3
    Uniform Sprite Size=30
    UsePositionOffset=false
    Write Material Random=true
   )
   &SkeletalMeshLocation(MaskSpawnGroup=false, Mesh Sampling Type=::NewEnumerator5, OverrideRandomSeed=false, Skeletal Mesh=$User.Pickaxe)
   &ShapeLocation(
    Additional Quaternion Rotation=true
    Additional Yaw / Pitch / Roll=<0.75 0.5 0>
    Hemisphere Distribution=<0.5 1>
    MaskSpawnGroup=false
    Override Seed=false
    Rotation Mode=::NewEnumerator3
    Rotation Quaternion=$Particles.SkeletalMeshLocation.SampledRotation
    Sphere Radius=2
    UseOwnerScale=true
   )
   &CameraOffset(Camera Offset Amount=10)
  ]
  particleUpdateScript [
   &ParticleState()
   &ScaleSpriteSize(
    Uniform Curve Scale=2
    Uniform Curve Sprite Scale=DataInterface() {propertyValues {Curve {keys [
        {arriveTangent=2.7886, interpMode=RCIM_Cubic, leaveTangent=2.7886, tangentMode=RCTM_User, time=0, value=0}
        {interpMode=RCIM_Cubic, tangentMode=RCTM_User, time=1, value=1}
       ]}}}
   )
   &DynamicMaterialParameters(
    Index 0 Param 1=1
    Index 0 Param 2=1
    Module.Param0WriteEnabled=true
    Module.Param1WriteEnabled=true
    Module.Param2WriteEnabled=true
    Module.Param3WriteEnabled=true
   )
   &ScaleColor(
    Scale Alpha=&FloatFromCurve(FloatCurve=DataInterface() {propertyValues {Curve {keys [<at 0.2 1 RCIM_Cubic>,<at 1 0 RCIM_Cubic>]}}})
    Scale RGB=&VectorFromFloat(Value=&FloatFromCurve(FloatCurve=DataInterface() {propertyValues {Curve {keys [
          {arriveTangent=-43.0022, interpMode=RCIM_Cubic, leaveTangent=-43.0022, tangentMode=RCTM_User, time=0, value=100}
          {arriveTangent=-3.5877, interpMode=RCIM_Cubic, leaveTangent=-3.5877, tangentMode=RCTM_User, time=0.2, value=1}
          <at 1 0 RCIM_Cubic>
         ]}}}))
    ScaleA=true
    ScaleRGB=true
   )
   /SetVariables(
    Particles.CoronaColor:LinearColor=&MakeLinearColorFromVectorAndFloat(
     "Vector (RGB)"=&Lerp_Vector(A=<0 0.4 0.9>, B=<1 1 1>, Alpha=$User.LightVersion)
     "Float (Alpha)"=&Lerp_Float(A=0.25, B=0.5, Alpha=$User.LightVersion)
    )
    Particles.CoronaDynamic:Vector4f=&MakeVector4(
     X=&FloatFromCurve(FloatCurve=DataInterface() {propertyValues {Curve {keys [<at 0 1 RCIM_Cubic>,<at 0.25 0.5 RCIM_Cubic>]}}})
     Y=1
     Z=1
     W=1
    )
    Particles.CoronaSize:Vector2f=&Vector2DFromFloat(Value=10)
    Particles.SoftFlareSize:Vector2f=&Vector2DFromFloat(Value=5)
   ) {isSetParameters=true}
  ]
  rendererClasses [&NiagaraSpriteRendererProperties]
  renderers [
   {rendererClass=&NiagaraSpriteRendererProperties}
   {rendererClass=&NiagaraSpriteRendererProperties, rendererIndex=1}
   {rendererClass=&NiagaraSpriteRendererProperties, rendererIndex=2}
  ]
 }}
summary {
 systemName=NS_PinkTrooper_Clobber_Idle
 emitters [{emitterName=FlareComplex, rendererClasses [&NiagaraSpriteRendererProperties]}]
 userVariables [
  User.LightVersion {type=NiagaraFloat, defaultValue=0, description=""}
  User.Pickaxe {
   type=NiagaraDataInterfaceSkeletalMesh
   defaultValue {
    struct=NiagaraExt_VariableValue_DataInterface
    value {dataInterface=&self:NiagaraDataInterfaceSkeletalMesh_0}
   }
  }
 ]
}
